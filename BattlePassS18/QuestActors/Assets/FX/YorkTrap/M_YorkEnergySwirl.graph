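# Unlit, two-sided translucent material for Niagara mesh particles, drawing the energy swirl in the York trap quest effect.
# The Fade dynamic parameter slides a sharpened V-gradient band along the mesh, shaped by two one-axis SphereMasks that soften the U and V edges.
# Particle color tints the mask through MF_FX_EmissiveAdjustments, and particle alpha scales it for transparency.
# SplineThicken widens the strip geometry using the Width and Width Low parameters.
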Material /BattlePassS18/QuestActors/Assets/FX/YorkTrap/M_YorkEnergySwirl
graph {
 properties {
  bUsedWithNiagaraMeshParticles=true
  blendMode=BLEND_Translucent
  pixelDepthOffsetMode=PDOM_Legacy
  shadingModel=MSM_Unlit
  twoSided=true
  usedShadingModels=Unlit
 }
 nodes [
  n1 = DynamicParameter() {paramNames [Fade,Param2,Param3,Param4]}
  n2 = TextureCoordinate()
  n3 = Add(A=LinearInterpolate(Alpha=@n1.Fade) {constA=-1}, B=Add(B=@n2))
  n4 = Add(A=@n2, B=LinearInterpolate(Alpha=@n1.Fade) {constA=-1})
  n5 = Multiply(A=OneMinus(None=ComponentMask(None=@n3) {g=true}), B=ComponentMask(None=Power(Base=@n4) {constExponent=5}) {g=true})
  n6 = TextureCoordinate() {uTiling=0}
  n7 = Constant2Vector() {g=0.5}
  n8 = Constant() {r=0.5}
  n9 = Constant() {r=0.75}
  n10 = TextureCoordinate() {vTiling=0}
  n11 = Constant2Vector() {r=0.5}
  n12 = Constant() {r=0.45}
  n13 = Constant() {r=0.5}
  n14 = Multiply(
   A=Saturate(None=Multiply(A=@n5) {constB=10})
   B=Multiply(A=SphereMask(A=@n6, B=@n7, Radius=@n8, Hardness=@n9), B=SphereMask(A=@n10, B=@n11, Radius=@n12, Hardness=@n13))
  )
  n15 = ParticleColor()
  MaterialFunctionCall(Emissive=Multiply(A=@n14, B=@n15.RGB)) {materialFunction=&MF_FX_EmissiveAdjustments}
  MaterialFunctionCall(WidthBase=ScalarParameter() {defaultValue=5, parameterName=Width}, WidthTip=ScalarParameter() {defaultValue=1, parameterName=Width Low}) {materialFunction=&SplineThicken}
  Multiply(A=@n15.A, B=@n14)
 ]
}
