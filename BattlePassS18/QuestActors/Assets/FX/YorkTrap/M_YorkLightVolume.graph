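# Additive light-volume material for Niagara mesh particles in the YorkTrap quest effect.
# Emissive combines particle color with a faint Fresnel-sampled rainbow, modulated by a per-particle sine flicker that the Flicker dynamic parameter blends in.
# An alpha mask combines V gradients, a travelling sine pulse, panning world-space noise, Fresnel and particle alpha.
# A shading-path switch drops depth and near-camera fades on Forward and Mobile.
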Material /BattlePassS18/QuestActors/Assets/FX/YorkTrap/M_YorkLightVolume
graph {
 properties {bUsedWithNiagaraMeshParticles=true, blendMode=BLEND_Additive, pixelDepthOffsetMode=PDOM_Legacy, usedShadingModels=Default Lit}
 nodes [
  n1 = ParticleRandom()
  n2 = Multiply(A=Time(), B=Add(A=Multiply(A=@n1) {constB=0.1}))
  n3 = Multiply(A=Sine(None=@n2) {period=0.155}, B=Sine(None=@n2) {period=0.2})
  n4 = Sine(None=@n2) {period=0.43}
  n5 = Clamp(None=Multiply(A=Sine(None=@n2) {period=0.075}, B=Multiply(A=@n3, B=@n4))) {maxDefault=3}
  n6 = DynamicParameter() {defaultValue=<rgba 0.6 1 1 1>, paramNames [Size,RotateSpeed,Flicker,ColorJitter]}
  n7 = ParticleColor()
  n8 = MaterialFunctionCall(Brightness=Constant() {r=1}, Exponent=Constant() {r=1}, Camera Vector=CameraVectorWS()) {materialFunction=&MF_Fresnel_FX}
  n9 = Multiply(A=TextureSample(UVs=@n8.Fresnel) {texture=/Game/Global/Textures/Default/Masks_and_Gradients/FortniteRainbow}.RGB) {constB=0.3}
  MaterialFunctionCall(Emissive=Multiply(A=LinearInterpolate(A=@n5, Alpha=@n6.Flicker) {constA=1}, B=Add(A=@n7.RGB, B=@n9))) {materialFunction=&MF_FX_EmissiveAdjustments}
  n10 = Multiply(A=Time())
  n11 = ComponentMask(None=TextureCoordinate()) {g=true}
  n12 = Power(Base=OneMinus(None=@n11)) {constExponent=1.5}
  n13 = Power(Base=Sine(None=Add(A=@n10, B=@n12)) {period=0.2}) {constExponent=5}
  n14 = Sine(None=@n8.Fresnel)
  n15 = Add(A=@n14) {constB=0.2}
  n16 = Multiply(A=LinearInterpolate(A=Saturate(None=@n15), B=Power(Base=@n11)) {constAlpha=0.9}, B=Saturate(None=OneMinus(None=@n14)))
  n17 = Multiply(A=Multiply(A=Saturate(None=@n13), B=@n16)) {constB=2}
  n18 = Constant() {r=1024}
  n19 = Divide(A=WorldPosition() {worldPositionShaderOffset=WPT_ExcludeAllShaderOffsets}.XYZ, B=Multiply(A=Abs(None=@n18)) {constB=-1})
  n20 = Time()
  n21 = Add(
   A=AppendVector(A=ComponentMask(None=@n19) {r=true}, B=ComponentMask(None=@n19) {b=true})
   B=AppendVector(A=Multiply(A=@n20) {constB=0.1}, B=Multiply(A=@n20) {constB=-1})
  )
  n22 = Multiply(
   A=Multiply(A=Add(A=@n17, B=@n16), B=@n11)
   B=Add(A=TextureSample(UVs=@n21) {samplerSource=SSM_Wrap_WorldGroupSettings, texture=/Game/Global/Textures/Default/Noise/T_Tiling_Noise_01}.R) {constB=0.2}
  )
  n23 = DepthFade(
   Opacity=MaterialFunctionCall(Distance=ScalarParameter() {defaultValue=125, parameterName=NearCamFade}) {materialFunction=&Near_Camera_Fade}.Result
   FadeDistance=ScalarParameter() {defaultValue=5, parameterName=DepthFade}
  )
  n24 = Constant() {r=1}
  Multiply(A=@n7.A, B=Multiply(A=Multiply(A=@n22, B=@n8.Fresnel), B=ShadingPathSwitch(Default=@n23, Forward=@n24, Mobile=@n24)))
 ]
}
