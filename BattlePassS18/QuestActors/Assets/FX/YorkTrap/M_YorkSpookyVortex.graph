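# Translucent, unlit, two-sided Niagara mesh-particle material for the York trap's spooky vortex, showing scrolling ghost faces in green-to-teal.
# A curved-UV face texture brightens the gradient and, multiplied by face-distorted Perlin noise, sphere-mask edge fades, TopFadeOut and particle alpha, forms opacity.
# Dynamic parameters randomize offset and pan time, though RandomUOffset drives both axes, leaving RandomVOffset unused.
# SplineThicken widens thin geometry using Width parameters.
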Material /BattlePassS18/QuestActors/Assets/FX/YorkTrap/M_YorkSpookyVortex
graph {
 properties {
  bUsedWithNiagaraMeshParticles=true
  blendMode=BLEND_Translucent
  pixelDepthOffsetMode=PDOM_Legacy
  shadingModel=MSM_Unlit
  twoSided=true
  usedShadingModels=Unlit
 }
 nodes [
  n1 = TextureCoordinate()
  n2 = Power(Base=ComponentMask(None=@n1) {g=true})
  n3 = LinearInterpolate(A=Constant3Vector() {constant=<rgba 0.0447 1 0.5548 0>}, B=Constant3Vector() {constant=<rgba 0 0.4758 0.5 0>}, Alpha=Saturate(None=@n2))
  n4 = TextureCoordinate() {uTiling=0.75, vTiling=1.25}
  n5 = AppendVector(A=ComponentMask(None=@n4) {r=true}, B=Add(A=ComponentMask(None=@n4) {g=true}, B=@n2))
  n6 = DynamicParameter() {paramNames [RandomUOffset, RandomVOffset, RandomTime, TopFadeOut]}
  n7 = AppendVector(A=@n6.RandomUOffset, B=@n6.RandomUOffset)
  n8 = Multiply(A=Time(), B=@n6.RandomTime)
  n9 = TextureSample(UVs=Panner(Coordinate=Add(A=@n5, B=@n7), Time=@n8) {speedX=0.25, speedY=1}) {texture=/Game/Effects/Fort_Effects/Textures/Tile/T_SpookyFaces_Chroma_Tile}
  n10 = Constant() {r=-2}
  Add(A=@n3, B=Multiply(A=@n3, B=Desaturation(None=@n9.RGB, Fraction=@n10)))
  MaterialFunctionCall(WidthBase=ScalarParameter() {defaultValue=100, parameterName=Width}, WidthTip=ScalarParameter() {defaultValue=5, parameterName=Width Low}) {materialFunction=&SplineThicken}
  n11 = Multiply(A=@n9.R) {constB=3}
  n12 = Add(A=Multiply(A=@n9.R) {constB=0.1}, B=TextureCoordinate() {uTiling=0, vTiling=0.5})
  n13 = Subtract(A=TextureSample(UVs=Panner(Coordinate=@n12, Time=@n8) {speedY=0.25}) {texture=/Game/Global/Textures/Default/Noise/3xPerlin_M}.G) {constA=0.25, constB=0.1}
  n14 = TextureCoordinate() {vTiling=0}
  n15 = Constant2Vector() {r=0.5}
  n16 = Constant() {r=0.45}
  n17 = Constant() {r=0.2}
  n18 = Multiply(
   A=SphereMask(A=TextureCoordinate() {uTiling=0}, B=Constant2Vector() {g=0.5}, Radius=Constant() {r=0.5}, Hardness=Constant() {r=0.75})
   B=Multiply(A=SphereMask(A=@n14, B=@n15, Radius=@n16, Hardness=@n17), B=Multiply(A=@n11, B=@n13))
  )
  n19 = ComponentMask(None=Add(A=@n1, B=LinearInterpolate(Alpha=@n6.TopFadeOut) {constA=-1})) {g=true}
  Multiply(A=Saturate(None=Multiply(A=@n18, B=@n19)), B=ParticleColor().A)
 ]
}
