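# Translucent, unlit, two-sided Niagara mesh-particle material for the York trap's spooky vortex cone, with a green-to-teal emissive gradient along V.
# Tiled UVs, warped along V by a power curve and distorted by panning Perlin noise, scroll a spooky-faces texture.
# Fresnel and V-gradient masks, both eroded by noise, shape the alpha.
# The TopFadeOut dynamic parameter shifts a V fade, and particle alpha scales the result.
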Material /BattlePassS18/QuestActors/Assets/FX/YorkTrap/M_YorkSpookyVortexCone
graph {
 properties {
  bUsedWithNiagaraMeshParticles=true
  blendMode=BLEND_Translucent
  pixelDepthOffsetMode=PDOM_Legacy
  previewMesh=&/SM_GhostTrapLightVolume
  shadingModel=MSM_Unlit
  twoSided=true
  usedShadingModels=Unlit
 }
 nodes [
  MaterialFunctionCall() {materialFunction=&DebugTimeSine}
  n1 = TextureCoordinate()
  n2 = ComponentMask(None=@n1) {g=true}
  MaterialFunctionCall(Emissive=LinearInterpolate(A=Constant3Vector() {constant=<rgba 0.0447 1 0.5548 0>}, B=Constant3Vector() {constant=<rgba 0 0.4758 0.5 0>}, Alpha=Power(Base=@n2))) {materialFunction=&MF_FX_EmissiveAdjustments}
  n3 = Multiply(A=@n2) {constB=2}
  n4 = Multiply(A=@n1, B=Constant2Vector() {g=4, r=8})
  n5 = ComponentMask(None=@n4) {g=true}
  n6 = OneMinus(None=@n2)
  n7 = AppendVector(A=ComponentMask(None=@n4) {r=true}, B=Add(A=OneMinus(None=@n5), B=Power(Base=@n6) {constExponent=3}))
  n8 = TextureSample(UVs=Panner(Coordinate=Multiply(A=@n7) {constB=0.5}) {speedX=0.2, speedY=1}) {texture=/Game/Global/Textures/Default/Noise/3xPerlin_M}
  n9 = Add(A=@n1, B=LinearInterpolate(Alpha=DynamicParameter() {paramNames [TopFadeOut,Power,Param3,Param4]}.TopFadeOut) {constA=-1})
  n10 = Saturate(None=Power(Base=ComponentMask(None=@n9) {g=true}))
  n11 = TextureSample(UVs=Panner(Coordinate=Multiply(A=@n7) {constB=0.25}) {speedX=0.2, speedY=0.6}) {texture=&3xPerlin_M}
  n12 = Add(A=@n7, B=Multiply(A=AppendVector(A=@n11.R, B=@n11.R)) {constB=0.3})
  n13 = MaterialFunctionCall(Brightness=Constant() {r=1}, Exponent=Constant() {r=3}, Camera Vector=CameraVectorWS()) {materialFunction=&MF_Fresnel_FX}
  n14 = Multiply(
   A=TextureSample(UVs=Panner(Coordinate=@n12) {speedX=1, speedY=3}) {texture=/Game/Effects/Fort_Effects/Textures/Tile/T_SpookyFaces_Chroma_Tile}.R
   B=Saturate(None=Subtract(A=@n13.Fresnel, B=@n8.R))
  )
  n15 = Multiply(A=Saturate(None=Subtract(A=@n3, B=@n8.R)), B=Subtract(A=@n10, B=OneMinus(None=@n14)))
  Multiply(A=Saturate(None=Multiply(A=@n15, B=@n10)), B=ParticleColor().A)
 ]
}
